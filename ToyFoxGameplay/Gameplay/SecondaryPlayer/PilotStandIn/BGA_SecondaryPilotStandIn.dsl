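; BGA_SecondaryPilotStandIn is a ToyFox building actor that stands in for a secondary pilot, carrying a replicated character item definition broadcast on RepNotify.
; Once playlist data is ready, the server binds to the ToyFox game state component's event-state changes, if a hotfix flag allows.
; One event state starts the StateTree after a random, hotfixable delay.
; Other states stop it and clear the timer; EndPlay also clears the timer and unbinds.
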(event Building|EventOnBuildingActorInitialized
  (params (InitializationReason :type EFortBuildingInitializationReason%20Enum :default "StaticallyPlaced") (BuildingPersistentState :type EFortBuildingPersistentState%20Enum :default "Default"))
  (bind _returnvalue (Fortnite|Athena|GameState|GetGameStateAthena))
  (bind _returnvalue_2 (Actor|GetComponentbyClass _returnvalue "/ToyFoxGameplay/Gameplay/B_ToyFoxGameStateComponent.B_ToyFoxGameStateComponent_C"))
  ;[0.001577 0.15 0 0.5] OnBuildingActorInitialized: Start the state tree logic.
  (|CallParentFunction InitializationReason BuildingPersistentState)
  (Networking|SwitchHasAuthority
    (:Authority
      (Class|FortAsyncActionCurrentPlaylistReady|PlaylistDataReadyAsync
        (:OnReady
          (if (ScalableFloat|GetValueasBool (Variables|Hotfix|GetHotfix_StateTreeEnabled))
            (Utilities|IsValid _returnvalue
              (:Is%20Valid
                (Utilities|IsValid _returnvalue_2
                  (:Is%20Valid
                    (Default|BindEventtoOnEventStateChanged _returnvalue_2 (EventDispatchers|CreateEvent :!event Custom|HandleEventStateChanged))
                    (CallFunction|HandleEventStateChanged (Event|GetEventState _returnvalue_2))))))))))))

(event EventSetCharacterItemDef
  (params (CharacterItemDef :type Athena%20Character%20Item%20Definition%20Object%20Reference))
  ;[0.15 0.01816 0.133 0.5] SetCharacterItemDef: Update the character item definition this stand in pilot represents.
  (Networking|SwitchHasAuthority
    (:Authority
      (Variables|Transient|SetCharacterItemDefinition CharacterItemDef)
      (Networking|ForceNetUpdate))))

(event Custom|HandleEventStateChanged
  (params (EventState :type ToyFoxEventStateEnum%20Enum :default "NewEnumerator0"))
  ;[0.0181 0.1337 0.15 0.5] HandleEventStateChanged: Start / stop firing logic based on event state.
  (Networking|SwitchHasAuthority
    (:Authority
      (if (Utilities|Enum|Equal(Enum) EventState "NewEnumerator3")
        ;; Start state tree after a random delay.
        (bind _returnvalue (Utilities|Time|SetTimerbyEvent (EventDispatchers|CreateEvent :!event EnableStateTree) (Math|Random|RandomFloatinRange (ScalableFloat|GetValueatLevel (Variables|Hotfix|GetHotfix_StateTreeEnableDelayMinimum)) (ScalableFloat|GetValueatLevel (Variables|Hotfix|GetHotfix_StateTreeEnableDelayMaximum)))))
        (Variables|Transient|SetEnableStateTreeDelayTimerHandle _returnvalue)
        (else
          (AI|Logic|StopLogic (Variables|Default|GetStateTree) (Utilities|String|Append "ToyFox EvenState has changed to " (Utilities|String|EnumtoString EventState)))
          (Utilities|Time|ClearandInvalidateTimerbyHandle (Variables|Transient|GetEnableStateTreeDelayTimerHandle)))))))

(event EventEndPlay
  (params (EndPlayReason :type EEndPlayReason%20Enum :default "Destroyed"))
  (bind _returnvalue (Fortnite|Athena|GameState|GetGameStateAthena))
  (bind _returnvalue_2 (Actor|GetComponentbyClass _returnvalue "/ToyFoxGameplay/Gameplay/B_ToyFoxGameStateComponent.B_ToyFoxGameStateComponent_C"))
  ;[0.1302 0.07708 0 0.5] EndPlay: Clean up delegate.
  (Utilities|FlowControl|Sequence
    (:then_0
      (Networking|SwitchHasAuthority
        (:Authority
          (Utilities|Time|ClearandInvalidateTimerbyHandle (Variables|Transient|GetEnableStateTreeDelayTimerHandle))
          (Utilities|IsValid _returnvalue
            (:Is%20Valid
              (Utilities|IsValid _returnvalue_2
                (:Is%20Valid
                  (Default|UnbindEventfromOnEventStateChanged _returnvalue_2 (EventDispatchers|CreateEvent :!event Custom|HandleEventStateChanged)))))))))
    (:then_1
      (|CallParentFunction EndPlayReason))))

(fn EnableStateTree
  ;[0 7.5e-05 0.15 0.5] EnableStateTree: Starts the brain component.
  (Networking|SwitchHasAuthority
    (:Authority
      (AI|Logic|StartLogic (Variables|Default|GetStateTree)))))

(fn GetCharacterItemDef
  (return (Variables|Transient|GetCharacterItemDefinition)))

(fn GetCharacterItemDefinition
  (return (Variables|Transient|GetCharacterItemDefinition)))

(fn OnRep_CharacterItemDefinition
  (Default|CallOnCharacterItemDefinitionUpdated (Variables|Transient|GetCharacterItemDefinition)))

(fn ConstructionScript)
