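; Blueprint interface for a secondary pilot stand-in in ToyFox gameplay's secondary player system.
; It declares a getter and a setter for an Athena Character Item Definition, the cosmetic character asset associated with the stand-in.
; The functions are declarations only, leaving storage and use of the definition to implementing Blueprints.
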(fn GetCharacterItemDef
  (return 0))

(fn SetCharacterItemDef
  (params (CharacterItemDef :type Athena%20Character%20Item%20Definition%20Object%20Reference)))
