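; Projectile Blueprint for the basic explosive rocket fired by ToyFox secondary-pilot stand-ins, judging by its name and graph.
; It overrides the fire-sound event to always play the third-person fire sound at the actor's location, since stand-in pilots never need a first-person variant.
; The construction script only calls its parent, so all other behavior shown here is left to the parent class.
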(event EventPlayFireSound
  ; StandIn pilots should always just play a 3p fire sound
  (Audio|PlaySoundatLocation (Variables|Default|GetFire%20Sound%203p) (Transformation|GetActorLocation)))

(fn ConstructionScript
  (|CallParentFunction))
