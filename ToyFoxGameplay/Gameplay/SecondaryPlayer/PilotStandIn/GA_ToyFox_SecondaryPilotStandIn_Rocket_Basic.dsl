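; Gameplay ability that lets secondary-player stand-in pilots fire a basic rocket from the giant pawn's head socket.
; On activation it applies a randomized attempt cooldown and runs an EQS query for a target.
; It aims at the target's bounds, predicted pawn position or head bone with random spread, then spawns the rocket through the giant's rocket component.
; A successful shot applies a shared giant cooldown that tag-gates activation.
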(event Ability|EventActivateAbility
  (bind _giantpawn (Variables|Transient|GetGiantPawn))
  ;[0 0.15 0.1266 0.5] ActivateAbility: When the ability activates, run an EQS to find the target destination to fire at.
  (Utilities|FlowControl|Sequence
    (:then_0
      (Variables|Transient|SetGiantPawn (Giant|GetGiantPawn (Actor|GetComponentbyClass (Game|GetGameState) "/ToyFoxGameplay/Gameplay/B_ToyFoxGameStateComponent.B_ToyFoxGameStateComponent_C"))))
    (:then_1
      (Utilities|IsValid (Ability|GetAvatarActorfromActorInfo)
        (:Is%20Valid
          (Utilities|IsValid _giantpawn
            (:Is%20Valid
              (bind _returnvalue (AI|EQS|RunEQSQuery (Variables|Settings|GetEnvQuery) _giantpawn (Variables|Settings|GetEQSRunMode)))
              (Utilities|IsValid _returnvalue
                (:Is%20Valid
                  (Variables|Transient|SetEqsQueryInstance _returnvalue)
                  (AI|EQS|BindEventtoOnQueryFinished _returnvalue (EventDispatchers|CreateEvent :!event Custom|HandleQueryFinished))
                  ;; Apply attempt cooldown so we're not constantly attempting the ability.
                  (Ability|ApplyAttemptCooldown))
                (:Is%20Not%20Valid
                  _)))
            (:Is%20Not%20Valid
              _))
          (Ability|EndAbility))
        (:Is%20Not%20Valid
          _)))))

(event Custom|HandleQueryFinished
  (params (QueryInstance :type EQS%20Query%20Instance%20Object%20Reference) (QueryStatus :type EEnvQueryStatus%20Enum :default "Processing"))
  (bind _returnvalue (Actor|GetComponentbyClass (Variables|Transient|GetGiantPawn) "/ToyFoxGameplay/Gameplay/SecondaryPlayer/B_ToyFoxGiantPawnRocketComponent.B_ToyFoxGiantPawnRocketComponent_C"))
  (bind _giantpawn (Variables|Transient|GetGiantPawn))
  ;[0.005135 0.007587 0.15 0.5] HandleQueryFinished: Fire projectile at the target destination.
  (Utilities|FlowControl|Sequence
    (:then_0
      (Utilities|IsValid QueryInstance
        (:Is%20Valid
          (bind (_bvalid _targetdestinationlocation) (Targeting|GetTargetDestinationfromQueryInstance QueryInstance))
          (if _bvalid
            (bind _asb_toy_fox_giant_player_pawn (Utilities|Casting|CastToB_ToyFoxGiantPlayerPawn _giantpawn))
            (bind _head_socket_transform (General|GetHeadSocketTransform _asb_toy_fox_giant_player_pawn))
            (bind _head_socket_transform_location (Math|Transform|BreakTransform _head_socket_transform))
            (Utilities|IsValid _returnvalue
              (:Is%20Valid
                (bind _explosiverocketprojectile (Class|BToyFoxGiantPawnRocketComponent|RequestSpawnBasicRocket _returnvalue _targetdestinationlocation _giantpawn _head_socket_transform_location "0, 0, 0" true))
                (Utilities|IsValid _explosiverocketprojectile
                  (:Is%20Valid
                    (Ability|CommitAbilityCooldown false)
                    (Ability|ApplyGameplayEffectToTarget (Ability|TargetData|AbilityTargetDatafromActor _giantpawn) (Variables|Settings|GetAbilityUseGiantCooldownGameplayEffect))
                    (Ability|ApplyGameplayEffectToOwner (Variables|Settings|GetRecentlyUsedGameplayEffectClass))))))))))
    (:then_1
      (Ability|EndAbility))))

(event Ability|EventOnEndAbility
  (params (bWasCancelled :type Boolean))
  ;[0.05729 0.03211 0 0.5] OnEndAbility: Clean up delegates.
  (Utilities|FlowControl|Sequence
    (:then_0
      (bind _eqsqueryinstance (Variables|Transient|GetEqsQueryInstance)
        (:then
          (AI|EQS|UnbindEventfromOnQueryFinished _eqsqueryinstance (EventDispatchers|CreateEvent :!event Custom|HandleQueryFinished)))
        (:else)))
    (:then_1
      (|CallParentFunction bWasCancelled))))

(fn ApplyAttemptCooldown
  (bind _returnvalue (Ability|MakeOutgoingGameplayEffectSpec (Variables|Settings|GetAbilityAttemptCooldownGameplayEffect)))
  (bind _output_get (LocalVariables|SetEffectSpec _returnvalue))
  (bind _returnvalue_2 (Ability|GameplayEffect|AssignTagSetbyCallerMagnitude _output_get "(TagName=\"SetByCaller.Duration\")" (Math|Random|RandomFloatinRange (ScalableFloat|GetCachedScalableFloatValue (Variables|Hotfix|GetHotfix_AttemptCooldownMin)) (ScalableFloat|GetCachedScalableFloatValue (Variables|Hotfix|GetHotfix_AttemptCooldownMax)))))
  (Ability|ApplyGameplayEffectSpecToOwner _returnvalue_2))

(fn GetGiantPawn
  (Utilities|FlowControl|Sequence
    (:then_0
      (bind _returnvalue (Fortnite|Athena|GameState|GetGameStateAthena))
      (Utilities|IsValid _returnvalue
        (:Is%20Valid
          (bind _returnvalue_2 (Actor|GetComponentbyClass _returnvalue "/ToyFoxGameplay/Gameplay/B_ToyFoxGameStateComponent.B_ToyFoxGameStateComponent_C"))
          (Utilities|IsValid _returnvalue_2
            (:Is%20Valid
              (return (Giant|GetGiantPawn _returnvalue_2)))))))
    (:then_1
      (return 0))))

(fn GetTargetDestinationFromQueryInstance
  (params (EqsQueryInstance :type EQS%20Query%20Instance%20Object%20Reference))
  ;[0 0.15 0.1316 0.5] Return the target location at which to aim the rocket.
  (Utilities|FlowControl|Sequence
    (:then_0
      (Utilities|IsValid EqsQueryInstance
        (:Is%20Valid
          (bind (_resultactors _returnvalue) (AI|EQS|GetQueryResultsasActors EqsQueryInstance))
          (if _returnvalue
            (if (Utilities|Array|IsValidIndex _resultactors)
              (Utilities|IsValid (Utilities|Array|Get(acopy) _resultactors)
                (:Is%20Valid
                  (Utilities|FlowControl|Sequence
                    (:then_0
                      (LocalVariables|SetTargetDestination (Collision|GetActorBounds (Utilities|Array|Get(acopy) _resultactors) true false))
                      (bind _asfort_pawn (Utilities|Casting|CastToFortPawn (Utilities|Array|Get(acopy) _resultactors))
                        (:then
                          (LocalVariables|SetTargetDestination (FortPawn|GetMovementPredictionPosition _asfort_pawn (ScalableFloat|GetCachedScalableFloatValue (Variables|Hotfix|GetHotfix_PredictionDeltaTime)))))
                        (:CastFailed
                          (bind _returnvalue_2 (Actor|GetComponentbyClass (Utilities|Array|Get(acopy) _resultactors) "/Script/Engine.SkeletalMeshComponent")) ; Primarily set up to work against the queen boss.
                          (Utilities|IsValid _returnvalue_2
                            (:Is%20Valid
                              (if (>= (Components|SkinnedMesh|GetBoneIndex _returnvalue_2 "head") 0)
                                (LocalVariables|SetTargetDestination (Transformation|GetSocketLocation _returnvalue_2 "head"))))))))
                    (:then_1
                      (return true (+ (LocalVariables|GetTargetDestination) (* (Math|Random|RandomUnitVector) (Math|Random|RandomFloatinRange 0.0 (ScalableFloat|GetCachedScalableFloatValue (Variables|Hotfix|GetHotfix_AccuracyOffsetMax)))))))))))))))
    (:then_1
      (return false "0, 0, 0"))))

(fn CanActivateAbility
  (params (ActorInfo :type Gameplay%20Ability%20Actor%20Info%20Structure) (Handle :type Gameplay%20Ability%20Spec%20Handle%20Structure))
  ;[0.02963 0.141 0.15 0.5] CanActivateAbility: The ability can only be activated if there's currently no general cooldown on the giant for the standin pilots to fire their rockets.
  (Utilities|FlowControl|Sequence
    (:then_0
      (bind _pawn (Helper|GetGiantPawn))
      (if (not (GameplayTags|HasAnyMatchingGameplayTags _pawn (Variables|Settings|GetGiantCanNotActivateAbilityGameplayTagContainer)))
        (return 0 true)))
    (:then_1
      (return 0 false))))
